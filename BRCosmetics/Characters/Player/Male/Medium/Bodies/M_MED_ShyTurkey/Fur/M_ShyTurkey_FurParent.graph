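# Parent fur material for the ShyTurkey player character body, built on the shared 2023 character and fur parent functions.
# CharacterParent PreFX and PostFX functions supply the character layers, with named reroutes carrying material, mask and tech art data.
# A channel-masked pre-shell adjustment multiplies base color and biases metallic, specular and roughness.
# That result chains through custom color, painted fur and shell functions, then a WPO shrink driven by fur length.
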Material /BRCosmetics/Characters/Player/Male/Medium/Bodies/M_MED_ShyTurkey/Fur/M_ShyTurkey_FurParent
graph {
 properties {
  bUseMaterialAttributes=true
  bUsedWithInstancedStaticMeshes=true
  bUsedWithMeshDeformer=true
  bUsedWithMorphTargets=true
  bUsedWithSkeletalMesh=true
  pixelDepthOffsetMode=PDOM_Legacy
  shadingModel=MSM_FromMaterialExpression
  usedShadingModels=Default Lit | Subsurface | Clear Coat | Subsurface Profile | Cloth
 }
 nodes [
  BlendMaterialAttributes()
  n1 = MaterialFunctionCall() {materialFunction=&MF_CharacterParent_2023_PreFX}
  MaterialFunctionCall(Material=NamedRerouteUsage(), M=@n1.M, Pre Skinned Local Position=@n1.M, Distance Blend=@n1.M, Fresnel=@n1.M) {materialFunction=&MF_CharacterParent_2023_PostFX}
  NamedRerouteDeclaration(None=@n1.Tech Art Mask) {name=TechArtMask, nodeColor=<rgba 0.2188 0 1 1>}
  NamedRerouteDeclaration(None=@n1.Result) {name=Material, nodeColor=<rgba 0 1 0.1094 1>}
  NamedRerouteDeclaration(None=@n1.M) {name=Mask, nodeColor=<rgba 1 0 0.2109 1>}
  n2 = GetMaterialAttributes(None=NamedRerouteUsage()) {attributeGetTypes [
    #1
    #2
    #3
    #4
   ]}
  n3 = ScalarParameter() {defaultValue=1, parameterName=PreFurShell_FurColorTipMulitplier}
  n4 = ChannelMaskParameter(None=NamedRerouteUsage()) {maskChannel=Blue, parameterName=PreFurShell_TexInputAdjustChannel, defaultValue=<rgba 0 0 1 0>}
  n5 = Add(A=@n2.Metallic, B=ScalarParameter() {parameterName=PreFurShell_MetallicBias})
  n6 = Add(A=@n2.Specular, B=ScalarParameter() {parameterName=PreFurShell_SpecularBias})
  n7 = Add(A=@n2.Roughness, B=ScalarParameter() {parameterName=PreFurShell_RoughnessBias})
  n8 = SetMaterialAttributes!(
   MaterialAttributes=@n2.MaterialAttributes
   Base Color=LinearInterpolate(A=@n2.Base Color, B=Multiply(A=@n2.Base Color, B=@n3), Alpha=@n4.R)
   Metallic=LinearInterpolate(A=@n2.Metallic, B=Saturate(None=@n5), Alpha=@n4.R)
   Specular=LinearInterpolate(A=@n2.Specular, B=Saturate(None=@n6), Alpha=@n4.R)
   Roughness=LinearInterpolate(A=@n2.Roughness, B=Saturate(None=@n7), Alpha=@n4.R)
  ) {attributeSetTypes [
    #1
    #2
    #3
    #4
   ]}
  n9 = NamedRerouteUsage()
  n10 = MaterialFunctionCall!(
   In=MaterialFunctionCall!(
    In=MaterialFunctionCall!(Material=@n8, TechArtMask=@n9) {materialFunction=/Game/Global/Materials/Character_Parent_2023/Fur_Parent_2023/MF_Fur_CustomColor}.Result
    TechArtMask=NamedRerouteUsage()
    Mask Input=NamedRerouteUsage()
   ) {materialFunction=/Game/Global/Materials/Character_Parent_2023/Fur_Parent_2023/MF_Fur_Painted}.Result
   TechArtMask=NamedRerouteUsage()
  ) {materialFunction=/Game/Global/Materials/Character_Parent_2023/Fur_Parent_2023/MF_Fur_Shells}
  NamedRerouteDeclaration(None=MaterialFunctionCall!(Material=@n10.Result, Mask=NamedRerouteUsage(), Fur Length=@n10.Result) {materialFunction=/Game/Global/Materials/Character_Parent_2023/Fur_Parent_2023/MF_FurWPOShrink}.Result) {name=OUT_Material, nodeColor=<rgba 0 0.5312 1 1>}
 ]
}
